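# Translucent UI material for the frontend interaction indicator divider line, building a procedural rectangular mask with no textures.
# UVs and inverted UVs, each offset by TranslateEdge, are scaled and saturated into four edge ramps, then combined with scalar dot products.
# A shared resolution sharpness correction function scales the Sharpness parameter to set ramp steepness from native pixel size.
# The vertical-axis ramps get an extra eightfold multiplier, making those edges harder than the horizontal ones.
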Material /Game/Athena/UI/Frontend/Spatial/Art/InteractionIndicators/M_UI_InteractionIndicator_DividerLine
graph {
 properties {
  bUsedWithStaticMesh=false
  blendMode=BLEND_Translucent
  materialDomain=MD_UI
  pixelDepthOffsetMode=PDOM_Legacy
  usedShadingModels=Default Lit
 }
 nodes [
  Constant() {r=1}
  n1 = TextureCoordinate()
  n2 = ScalarParameter() {defaultValue=-0.0125, parameterName=TranslateEdge}
  n3 = Add(A=OneMinus(None=@n1), B=@n2)
  n4 = Multiply(
   A=ScalarParameter() {defaultValue=8, parameterName=Sharpness}
   B=MaterialFunctionCall!(Native_U_PixelSize=ScalarParameter() {defaultValue=6, parameterName=Native_Pixel_Size}) {materialFunction=/FortUILibrary/Foundation/Art_ToolKit/Tools/M_UI_Resolution_Sharpness_Correction}.Result
  )
  n5 = Multiply(A=@n4) {constA=2, constB=8}
  n6 = Saturate(None=Multiply(A=ComponentMask(None=@n3) {g=true}, B=@n5))
  n7 = Add(A=@n1, B=@n2)
  n8 = Saturate(None=Multiply(A=ComponentMask(None=@n7) {g=true}, B=@n5))
  n9 = ScalarParameter() {defaultValue=-0.0125, parameterName=TranslateEdge}
  n10 = Add(A=OneMinus(None=@n1), B=@n9)
  n11 = Saturate(None=Multiply(A=ComponentMask(None=@n10) {r=true}, B=@n4))
  n12 = Add(A=@n1, B=@n9)
  n13 = Saturate(None=Multiply(A=ComponentMask(None=@n12) {r=true}, B=@n4))
  Saturate(None=DotProduct(A=DotProduct(A=@n6, B=@n8), B=DotProduct(A=@n11, B=@n13)))
 ]
}
