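# Translucent UI master material for glitchy, hologram-style point-of-interest indicators, named for text, with an optional icon-mask mode.
# Summed, time-scaled sines make scrolling scanlines, layered with a dot grid and two panning square patterns whose frequencies jump in rounded, sine-driven steps.
# That pattern drives color and opacity, with optional edge fading and an outline mode that alternates scanlines by UV1.
# Static switches select RGB icon-mask opacity and core/glow coloring.
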Material /Game/Athena/UI/Frontend/Spatial/Art/InteractionIndicators/M_UI_POI_Glitch_Text_Master
graph {
 properties {
  bUsedWithStaticMesh=false
  blendMode=BLEND_Translucent
  materialDomain=MD_UI
  numCustomizedUVs=2
  pixelDepthOffsetMode=PDOM_Legacy
  usedShadingModels=Default Lit
 }
 nodes [
  n1 = Multiply(
   A=Add(A=Time(), B=ScalarParameter() {group=Global, parameterName=GlobalTimeOffset})
   B=ScalarParameter() {defaultValue=0.5, group=Global, parameterName=GlobalTimeControl}
  )
  n2 = Multiply(A=ScalarParameter() {defaultValue=0.8, group=ScanLines, parameterName=ScanLines_Speed}) {constB=-1}
  n3 = TextureCoordinate()
  n4 = ScalarParameter() {defaultValue=1, group=ScanLines, parameterName=ScanLines_SimpleScale}
  n5 = ComponentMask(None=Add(A=Multiply(A=@n3, B=@n4), B=Multiply(A=@n1, B=@n2))) {g=true}
  n6 = ScalarParameter() {defaultValue=4, group=ScanLines, parameterName=ScanLines_Softness}
  n7 = Multiply(A=Sine(None=@n5) {period=0.666}, B=@n6)
  n8 = Multiply(A=Sine(None=@n5) {period=0.2}, B=@n6)
  n9 = Sine(None=@n5) {period=0.325}
  n10 = LinearInterpolate(
   A=LinearInterpolate(A=Clamp(None=@n7) {minDefault=-1}, Alpha=Clamp(None=@n8) {minDefault=-1})
   Alpha=Clamp(None=Multiply(A=@n9, B=@n6)) {minDefault=-1}
  )
  n11 = Sine(None=@n5) {period=0.525}
  n12 = OneMinus(None=Clamp(None=Multiply(A=@n11, B=@n6)) {minDefault=-1})
  n13 = Multiply(A=Sine(None=@n5) {period=0.2}, B=@n6)
  n14 = Saturate(None=DotProduct(A=DotProduct(A=@n10, B=@n12), B=Clamp(None=@n13) {minDefault=-1}))
  Add(A=TextureCoordinate(), B=Multiply(A=@n14, B=ScalarParameter() {defaultValue=0.05, parameterName=DistortionAmount}))
  n15 = TextureSampleParameter2D(UVs=MaterialFunctionCall(UVs=TextureCoordinate(), Texture Scale=ScalarParameter() {defaultValue=1, group=Icon, parameterName=IconScale}) {materialFunction=&ScaleUVsByCenter}.UVs) {group=Icon, parameterName=IconRGBMask, texture=/Game/Athena/UI/Frontend/Spatial/Art/T_UI_AlterLogo}
  n16 = Multiply(A=ScalarParameter() {defaultValue=0.075, group=Icon, parameterName=OutlineThickness, sliderMin=1}) {constB=-1}
  n17 = Saturate(None=Multiply(A=Add(A=@n15.R, B=@n16) {constB=-0.075}, B=ScalarParameter() {defaultValue=32, group=Icon, parameterName=IconSharpness}) {constB=32})
  n18 = MaterialFunctionCall() {materialFunction=/FortUILibrary/Foundation/Art_ToolKit/Tools/MF_UI_SquareToScreenSpace}
  n19 = ScalarParameter() {defaultValue=512, group=Dots, parameterName=Dots_Size}
  n20 = ComponentMask(None=Sine(None=Multiply(A=@n18.Y_Priority, B=@n19))) {g=true}
  n21 = ScalarParameter() {defaultValue=16, group=Dots, parameterName=Dots_Softness}
  n22 = ComponentMask(None=Sine(None=Multiply(A=@n18.X_Priority, B=@n19))) {r=true}
  n23 = DotProduct(A=Saturate(None=Multiply(A=@n20, B=@n21)), B=Saturate(None=Multiply(A=@n22, B=@n21)))
  n24 = ScalarParameter() {defaultValue=0.5, group=Squares, parameterName=SimpleScaleChange}
  n25 = Multiply(A=ScalarParameter() {defaultValue=16, group=Squares, parameterName=V_Min_Offset}, B=@n24)
  n26 = Multiply(A=ScalarParameter() {defaultValue=24, group=Squares, parameterName=V_Max_Offset}, B=@n24)
  n27 = Multiply(
   A=Add(A=Time(), B=ScalarParameter() {group=Global, parameterName=GlobalTimeOffset})
   B=ScalarParameter() {defaultValue=0.5, group=Global, parameterName=GlobalTimeControl}
  )
  n28 = ScalarParameter() {defaultValue=0.25, group=Squares, parameterName=V_Time_Sine_1}
  n29 = Add(A=Sine(None=Multiply(A=@n27, B=@n28)))
  n30 = ScalarParameter() {defaultValue=4, group=Squares, parameterName=V_Time_Sine_2}
  n31 = Add(A=Sine(None=Multiply(A=@n27, B=@n30)))
  n32 = Sine(None=Multiply(A=@n27, B=ScalarParameter() {defaultValue=1, group=Squares, parameterName=V_Time_Sine_3}))
  n33 = DotProduct(A=DotProduct(A=LinearInterpolate(Alpha=@n29), B=LinearInterpolate(Alpha=@n31)), B=LinearInterpolate(Alpha=Add(A=@n32)))
  n34 = AppendVector(A=Constant(), B=Round(None=LinearInterpolate(A=@n25, B=@n26, Alpha=@n33)))
  n35 = MaterialFunctionCall() {materialFunction=&MF_UI_SquareToScreenSpace}
  n36 = Panner(Coordinate=@n35.Y_Priority) {speedY=0.25}
  n37 = ComponentMask(None=Sine(None=Multiply(A=@n34, B=@n36))) {g=true}
  n38 = ScalarParameter() {defaultValue=8, group=Squares, parameterName=Squares_Softness}
  n39 = Panner(Coordinate=@n35.X_Priority) {speedX=-0.025}
  n40 = Multiply(A=ScalarParameter() {defaultValue=16, group=Squares, parameterName=U_Min_Offset}, B=@n24)
  n41 = Multiply(A=ScalarParameter() {defaultValue=24, group=Squares, parameterName=U_Max_Offset}, B=@n24)
  n42 = ScalarParameter() {defaultValue=0.3, group=Squares, parameterName=U_Time_Sine_1}
  n43 = Add(A=Sine(None=Multiply(A=@n27, B=@n42)))
  n44 = ScalarParameter() {defaultValue=2, group=Squares, parameterName=U_Time_Sine_2}
  n45 = Add(A=Sine(None=Multiply(A=@n27, B=@n44)))
  n46 = Sine(None=Multiply(A=@n27, B=ScalarParameter() {defaultValue=0.75, group=Squares, parameterName=U_Time_Sine_3}))
  n47 = DotProduct(A=DotProduct(A=LinearInterpolate(Alpha=@n43), B=LinearInterpolate(Alpha=@n45)), B=LinearInterpolate(Alpha=Add(A=@n46)))
  n48 = AppendVector(A=Round(None=LinearInterpolate(A=@n40, B=@n41, Alpha=@n47)), B=Constant())
  n49 = ComponentMask(None=Sine(None=Multiply(A=@n39, B=@n48))) {r=true}
  n50 = DotProduct(A=Saturate(None=Multiply(A=@n37, B=@n38)), B=Saturate(None=Multiply(A=@n49, B=@n38)))
  n51 = Multiply(A=@n25) {constB=0.5}
  n52 = Multiply(A=@n26) {constB=0.5}
  n53 = AppendVector(A=Constant(), B=Round(None=LinearInterpolate(A=@n51, B=@n52, Alpha=@n33)))
  n54 = MaterialFunctionCall() {materialFunction=&MF_UI_SquareToScreenSpace}
  n55 = Panner(Coordinate=@n54.Y_Priority) {speedY=-2}
  n56 = ComponentMask(None=Sine(None=Multiply(A=@n53, B=@n55))) {g=true}
  n57 = ScalarParameter() {defaultValue=8, group=Squares, parameterName=Squares_Softness}
  n58 = Panner(Coordinate=@n54.X_Priority) {speedX=0.5}
  n59 = Multiply(A=@n40) {constB=0.5}
  n60 = Multiply(A=@n41) {constB=0.5}
  n61 = AppendVector(A=Round(None=LinearInterpolate(A=@n59, B=@n60, Alpha=@n47)), B=Constant())
  n62 = ComponentMask(None=Sine(None=Multiply(A=@n58, B=@n61))) {r=true}
  n63 = DotProduct(A=Saturate(None=Multiply(A=@n56, B=@n57)), B=Saturate(None=Multiply(A=@n62, B=@n57)))
  n64 = Add(A=@n50, B=@n63)
  n65 = Add(
   A=Multiply(A=@n23, B=ScalarParameter() {defaultValue=0.1, group=Dots, parameterName=Dots_Opacity})
   B=LinearInterpolate(
    A=ScalarParameter() {defaultValue=0.5, group=Squares, parameterName=Squares_OpacityMin}
    B=ScalarParameter() {defaultValue=0.6, group=Squares, parameterName=Squares_OpacityMax}
    Alpha=Saturate(None=@n64)
   )
  )
  n66 = Multiply(A=@n14, B=ScalarParameter() {defaultValue=0.25, group=ScanLines, parameterName=ScanLines_Opacity})
  n67 = Saturate(None=Add(A=Saturate(None=@n65), B=@n66))
  n68 = OneMinus(None=@n17)
  n69 = OneMinus(None=@n14)
  n70 = Floor(None=ComponentMask(None=TextureCoordinate() {coordinateIndex=1, uTiling=2}) {r=true})
  n71 = Saturate(None=LinearInterpolate(A=LinearInterpolate(B=@n67, Alpha=@n69), B=LinearInterpolate(B=@n67, Alpha=@n14), Alpha=Saturate(None=@n70)))
  n72 = Add(A=DotProduct(A=@n17, B=@n67), B=DotProduct(A=LinearInterpolate(B=@n68, Alpha=@n15.G), B=@n71))
  n73 = Add(
   A=@n72
   B=LinearInterpolate(B=DotProduct(A=@n15.B, B=@n67), Alpha=ScalarParameter() {defaultValue=1, group=Icon, parameterName=GlowAmount, sliderMin=1})
  )
  n74 = TextureCoordinate()
  n75 = Sine(None=@n74) {period=2}
  n76 = ComponentMask(None=@n74) {g=true}
  n77 = ComponentMask(None=Sine(None=@n74) {period=2}) {g=true}
  n78 = ScalarParameter() {parameterName=FadeAlledges}
  n79 = LinearInterpolate(A=@n67, B=@n71, Alpha=ScalarParameter() {group=Global, parameterName=IsOutline})
  n80 = DotProduct(
   A=Power(Base=ComponentMask(None=@n75) {r=true}, Exp=ScalarParameter() {defaultValue=1.5, parameterName=FalloffRate}) {constExponent=1.5}
   B=DotProduct(A=LinearInterpolate(A=@n76, B=@n77, Alpha=@n78), B=@n79)
  )
  n81 = ScalarParameter() {parameterName=HasFade}
  Multiply(
   A=StaticSwitchParameter(True=Saturate(None=@n73), False=LinearInterpolate(A=@n79, B=@n80, Alpha=@n81)) {group=Icon, parameterName=IsIconMask}
   B=ScalarParameter() {defaultValue=1, parameterName=ForceAlphaChange}
  )
  n82 = VectorParameter() {group=Squares, parameterName=Squares_Color_A, defaultValue=<rgba 0.0375 0.2156 0.75 1>}
  n83 = VectorParameter() {group=Squares, parameterName=Squares_Color_B, defaultValue=<rgba 0 0.5 1 1>}
  n84 = Saturate(None=DotProduct(A=@n50, B=@n63))
  StaticSwitchParameter(
   True=LinearInterpolate(
    A=VectorParameter() {group=Icon, parameterName=IconGlowColor, defaultValue=<rgba 1 1 1 1>}.RGB
    B=VectorParameter() {group=Icon, parameterName=IconCoreColor, defaultValue=<rgba 1 1 1 1>}.RGB
    Alpha=@n72
   )
   False=LinearInterpolate(
    A=LinearInterpolate(A=@n82.RGB, B=@n83.RGB, Alpha=@n84)
    B=VectorParameter() {group=Dots, parameterName=DotsColor, defaultValue=<rgba 0.05 0.525 1 1>}.RGB
    Alpha=Add(A=@n23, B=@n66)
   )
  ) {group=Icon, parameterName=UseIconGlowColors}
 ]
}
